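# Sub-heading widget for the matchmaking settings UI, built on CommonUserWidget.
# An Overlay root holds an auto-wrapping CommonTextBlock with the base text style, placeholder title text and bottom padding.
# A UIKit toggle block is also declared as a widget variable, with no parent shown in the Overlay hierarchy.
# The event graph is empty, so the asset itself contains no logic to set the text or drive the toggle.
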WidgetBlueprint /MatchmakingSettingsUI/MatchmakingSettings/SettingEntries/SectionHeading/WBP_MatchmakingSettings_SubHeading
graphs [EventGraph {links [], nodes []}]
variables []
widgets {
 info {inheritedWidgetCount=0, namedSlotCount=1, parentClass=&CommonUserWidget, rootWidgetClass=&Overlay, widgetCount=3}
 widgets [
  O_TextContainer {widgetClassPath=&Overlay, widget=O_TextContainer}
  Txt_Subheading {
   widgetClassPath=&CommonTextBlock
   bIsVariable=true
   parent=O_TextContainer
   widget=Txt_Subheading
   properties {
    autoWrapText=true
    style=&TextStyle-Base_C
    text=$Subheading Title
    font {
     bForceMonospaced=false
     bMaterialIsStencil=false
     fontMaterial=null
     fontObject=&NotoSans
     letterSpacing=0
     monospacedWidth=1
     size=24
     skewAmount=0
     typefaceFontName=NotoSans-Regular
     outlineSettings {
      bApplyOutlineToDropShadows=false
      bMiteredCorners=false
      bSeparateFillAlpha=false
      outlineMaterial=null
      outlineSize=0
      outlineColor=<rgba 0 0 0 1>
     }
    }
    shadowOffset=<0 0>
   }
   slot {class=OverlaySlot, properties {horizontalAlignment=HAlign_Fill, padding=<ltrb 0 0 0 12>}}
  }
  Block_Toggle {widgetClassPath=/FortUILibrary/Systems/ModularButton/Blocks/WBP_UIKit_Block_Toggle, bIsVariable=true, widget=Block_Toggle}
 ]
}
